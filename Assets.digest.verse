# Copyright Epic Games, Inc. All Rights Reserved.
#################################################
# Generated Digest of Verse API
# DO NOT modify this manually!
# Generated from build: ++Fortnite+Release-36.20-CL-44118662
#################################################


using {/Verse.org/Native}
using {/Verse.org/SceneGraph}
using {/Verse.org/Simulation}
using {/Verse.org/Colors}
using {/Verse.org/Assets}
# Module import path: /playkigai.dig@fortnite.com/Minigame_Spacewalk/Materials
Materials := module:
    M_CubeInner<scoped {/playkigai.dig@fortnite.com/Minigame_Spacewalk}> := class<final><scoped {/playkigai.dig@fortnite.com/Minigame_Spacewalk}>(material):
        @editable
        var BaseColor<public>:color = external {}

    M_CubeOuter<scoped {/playkigai.dig@fortnite.com/Minigame_Spacewalk}> := class<final><scoped {/playkigai.dig@fortnite.com/Minigame_Spacewalk}>(material):


# Module import path: /playkigai.dig@fortnite.com/Minigame_Spacewalk/Meshes
Meshes := module:
    SM_CharacterCollision<scoped {/playkigai.dig@fortnite.com/Minigame_Spacewalk}> := class<final><scoped {/playkigai.dig@fortnite.com/Minigame_Spacewalk}>(mesh_component):
        @editable
        var Material_0<public>:material = external {}

    SM_CharacterCollision_asset<scoped {/playkigai.dig@fortnite.com/Minigame_Spacewalk}>:mesh = external {}

    SM_CubeInner<scoped {/playkigai.dig@fortnite.com/Minigame_Spacewalk}> := class<final><scoped {/playkigai.dig@fortnite.com/Minigame_Spacewalk}>(mesh_component):
        @editable
        var Material<public>:material = external {}

    SM_CubeInner_asset<scoped {/playkigai.dig@fortnite.com/Minigame_Spacewalk}>:mesh = external {}

    SM_CubeOuter<scoped {/playkigai.dig@fortnite.com/Minigame_Spacewalk}> := class<final><scoped {/playkigai.dig@fortnite.com/Minigame_Spacewalk}>(mesh_component):
        @editable
        var Material_0<public>:material = external {}

    SM_CubeOuter_asset<scoped {/playkigai.dig@fortnite.com/Minigame_Spacewalk}>:mesh = external {}

    SM_DeathVolume<scoped {/playkigai.dig@fortnite.com/Minigame_Spacewalk}> := class<final><scoped {/playkigai.dig@fortnite.com/Minigame_Spacewalk}>(mesh_component):
        @editable
        var Material_0<public>:material = external {}

    SM_DeathVolume_asset<scoped {/playkigai.dig@fortnite.com/Minigame_Spacewalk}>:mesh = external {}

    SM_Respawn<scoped {/playkigai.dig@fortnite.com/Minigame_Spacewalk}> := class<final><scoped {/playkigai.dig@fortnite.com/Minigame_Spacewalk}>(mesh_component):
        @editable
        var Material_0<public>:material = external {}

    SM_Respawn_asset<scoped {/playkigai.dig@fortnite.com/Minigame_Spacewalk}>:mesh = external {}

    SM_StartingPlate<scoped {/playkigai.dig@fortnite.com/Minigame_Spacewalk}> := class<final><scoped {/playkigai.dig@fortnite.com/Minigame_Spacewalk}>(mesh_component):


    SM_StartingPlate_asset<scoped {/playkigai.dig@fortnite.com/Minigame_Spacewalk}>:mesh = external {}

# Module import path: /playkigai.dig@fortnite.com/Minigame_Spacewalk/Prefabs
Prefabs := module:
    @import_as("/Minigame_Spacewalk/Prefabs/character_collision_proxy.character_collision_proxy_C")
    character_collision_proxy<scoped {/playkigai.dig@fortnite.com/Minigame_Spacewalk}> := class<final>(entity):


    character_collision_proxy_asset<scoped {/playkigai.dig@fortnite.com/Minigame_Spacewalk}>:entity_prefab = external {}

    @import_as("/Minigame_Spacewalk/Prefabs/collision_proxy_spawner.collision_proxy_spawner_C")
    collision_proxy_spawner<scoped {/playkigai.dig@fortnite.com/Minigame_Spacewalk}> := class<final>(entity):


    collision_proxy_spawner_asset<scoped {/playkigai.dig@fortnite.com/Minigame_Spacewalk}>:entity_prefab = external {}

    @import_as("/Minigame_Spacewalk/Prefabs/death_volume.death_volume_C")
    death_volume<scoped {/playkigai.dig@fortnite.com/Minigame_Spacewalk}> := class<final>(entity):


    death_volume_asset<scoped {/playkigai.dig@fortnite.com/Minigame_Spacewalk}>:entity_prefab = external {}

    @import_as("/Minigame_Spacewalk/Prefabs/game_manager.game_manager_C")
    game_manager<scoped {/playkigai.dig@fortnite.com/Minigame_Spacewalk}> := class<final>(entity):


    game_manager_asset<scoped {/playkigai.dig@fortnite.com/Minigame_Spacewalk}>:entity_prefab = external {}

    @import_as("/Minigame_Spacewalk/Prefabs/gameplay_cube.gameplay_cube_C")
    gameplay_cube<scoped {/playkigai.dig@fortnite.com/Minigame_Spacewalk}> := class<final>(entity):


    gameplay_cube_asset<scoped {/playkigai.dig@fortnite.com/Minigame_Spacewalk}>:entity_prefab = external {}

    @import_as("/Minigame_Spacewalk/Prefabs/respawn.respawn_C")
    respawn<scoped {/playkigai.dig@fortnite.com/Minigame_Spacewalk}> := class<final>(entity):


    respawn_asset<scoped {/playkigai.dig@fortnite.com/Minigame_Spacewalk}>:entity_prefab = external {}

    @import_as("/Minigame_Spacewalk/Prefabs/start_game.start_game_C")
    start_game<scoped {/playkigai.dig@fortnite.com/Minigame_Spacewalk}> := class<final>(entity):


    start_game_asset<scoped {/playkigai.dig@fortnite.com/Minigame_Spacewalk}>:entity_prefab = external {}

    @import_as("/Minigame_Spacewalk/Prefabs/ui_manager.ui_manager_C")
    ui_manager<scoped {/playkigai.dig@fortnite.com/Minigame_Spacewalk}> := class<final>(entity):


    ui_manager_asset<scoped {/playkigai.dig@fortnite.com/Minigame_Spacewalk}>:entity_prefab = external {}