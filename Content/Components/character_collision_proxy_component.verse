<#
@UUID: 67e17241-d841-46ab-858f-15404380006a
@Task: Make a component to detect collisions between entities and agents.
@Description: Implement a SceneGraph component that detects collisions between the agent it is attached to and other entities.
@Tags: CollisionDetection, AgentCollisionDetection
@Export: true
#>

using { /Fortnite.com/Characters }
using { /UnrealEngine.com/Temporary/SpatialMath }
using { /Verse.org/Native }
using { /Verse.org/Simulation }
using { /Verse.org/SpatialMath }
using { /Verse.org/SceneGraph }
using { Prefabs }

# A component that acts as a proxy for the FortCharacter collision, since entities do not detect collision with anything other than entities.
character_collision_proxy_component<public> := class<final_super>(component):
    var MaybeAgent: ?agent = false
    var MaybePostTickHandler: ?cancelable = false
    var PrevCharPosition: (/UnrealEngine.com/Temporary/SpatialMath:)vector3 = (/UnrealEngine.com/Temporary/SpatialMath:)vector3{}


    SetAgent<public>(InAgent : agent):void=
        set MaybeAgent = option. InAgent


    OnBeginSimulation<override>():void=
        (super:)OnBeginSimulation()

        set MaybePostTickHandler = option. TickEvents.PostPhysics.Subscribe(OnPostPhysics)


    OnEndSimulation<override>():void=
        if (PostTickHandler := MaybePostTickHandler?):
            PostTickHandler.Cancel()


    # On every post physics, set the collision proxy's transform equal to the character's.
    OnPostPhysics(Delta:float):void=
        if (FortCharacter := MaybeAgent?.GetFortCharacter[]):
            CharacterTranslation    := FortCharacter.GetTransform().Translation
            TranslationDelta        := CharacterTranslation - PrevCharPosition
            TranslationPrediction   := CharacterTranslation + TranslationDelta

            Entity.SetGlobalTransform((/Verse.org/SpatialMath:)transform{ 
                Translation := FromVector3(TranslationPrediction),
                Rotation := Entity.GetGlobalTransform().Rotation, 
                Scale := Entity.GetGlobalTransform().Scale
            })

            set PrevCharPosition = CharacterTranslation


#Extend the character_collision_proxy with a GetAgent() function.
(CollisionProxy:character_collision_proxy).GetAgent<public>()<decides><transacts>:agent=
    ProxyComponent := CollisionProxy.GetComponent[character_collision_proxy_component]
    ProxyComponent.MaybeAgent?