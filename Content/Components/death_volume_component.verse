<#
@UUID: 6210f02f-bc49-4189-ab07-1326a4da0fd4
@Task: A component that detects when the player falls too low.
@Description: Implement a SceneGraph component that will teleport the player back to an available respawn point colliding with it.
@Tags: Collision, Gameplay
@Export: true
#>

using { /UnrealEngine.com/Temporary/SpatialMath }
using { /Fortnite.com/Characters }
using { /Verse.org/SceneGraph }
using { /Verse.org/Simulation }
using {Prefabs}

death_volume_component<public> := class<final_super>(component):

    var MaybeMesh:?mesh_component = false

    OnBeginSimulation<override>():void=
        if(Mesh := Entity.GetComponent[mesh_component]):
            set MaybeMesh = option. Mesh
            Mesh.EntityEnteredEvent.Subscribe(OnEntityEntered)

    OnEntityEntered(EnteredEntity:entity):void=
        if:
            Sim := Entity.GetSimulationEntity[]
            Proxy := character_collision_proxy[EnteredEntity] 
            Agent := Proxy.GetAgent[]
            Fort := Agent.GetFortCharacter[]
            Respawns := Sim.FindDescendantEntitiesWithComponent(respawn_component)
        then:
            TeleportPlayerToRespawn(Agent,Respawns)

    TeleportPlayerToRespawn(Agent:agent, Respawns:generator(entity)):void=
        var HasBeenTeleported:logic = false
        if(Fort := Agent.GetFortCharacter[]):
            for(Respawn : Respawns, Comp := Respawn.GetComponent[respawn_component]):
                if:
                    not HasBeenTeleported?
                    Comp.IsAvailable[]
                    Transform := Respawn.GetGlobalTransform()
                    Fort.TeleportTo[(/UnrealEngine.com/Temporary/SpatialMath:)vector3{
                            X := Transform.Translation.Forward,
                            Y := -Transform.Translation.Left,
                            Z := Transform.Translation.Up + 300.0
                        }, 
                        (/UnrealEngine.com/Temporary/SpatialMath:)IdentityRotation()
                    ]
                then:
                    set HasBeenTeleported = true