<#
@UUID: 49bef4fe-a2ec-41ee-a0bc-10d4e8d8670e
@Task: Make a component to manage the game state.
@Description: Implement a SceneGraph component that manages a timed player-versus-player control-the-most-cubes game.
@Tags: Game Manager
@Export: true
#>

using { /Fortnite.com/Devices }
using { /Verse.org/Colors }
using { /Verse.org/SceneGraph }
using { /Verse.org/Simulation }


# Class that manages the game state.
game_manager_component<public> := class<final_super>(component):
    @editable:
        ToolTip := GameLengthToolTip
        Categories := array{GameplayCategory}
    GameLength: int = 15


    @editable:
        ToolTip := GameplayCubesToolTip
        Categories := array{GameplayCategory}
    GameplayCubes: []Prefabs.gameplay_cube = array{}
    
    
    @editable:
        ToolTip := MutatorZonesToolTip
        Categories := array{GameplayCategory}
    MutatorZones: []mutator_zone_device = array{}
    
    
    @editable:
        ToolTip := UIManagerToolTip
        Categories :=  array{GameplayCategory}
    var MaybeUIManager: ?entity = false

    @editable:
        ToolTip := MessageDevicerToolTip
        Categories :=  array{GameplayCategory}
    var MaybeMessageDevice: ?hud_message_device = false


    @editable:
        ToolTip := P1ColorToolTip
        Categories := array{PlayerCustomizationCategory}
    P1Color: color = NamedColors.Red


    @editable:
        ToolTip := P2ColorToolTip
        Categories := array{PlayerCustomizationCategory}
    P2Color: color = NamedColors.Blue

    
    var CurrentGameTimer: int = 0
    var AgentToColorMap: [agent]?color = map{}
    var AgentToScoreMap: [agent]?int = map{}



    # === Entity Functions ===


    OnBeginSimulation<override>():void=
        (super:)OnBeginSimulation()

        set CurrentGameTimer = GameLength

        # Initialize cubes.
        for (CubePrefab : GameplayCubes):
            CubePrefab.Subscribe(OnCubeCollision)

        # Disable mutator zones until the game starts.
        for (MutatorZone : MutatorZones):
            MutatorZone.Disable()


    # === Manager Functions ===


    # Handle a cube changing owner.
    OnCubeCollision<private>(InAgent:agent, MaybeCurrentController:?agent, GameplayCubeRef:Prefabs.gameplay_cube):void=
        if (MaybeCurrentController? = InAgent):
            return

        if (MaybeCurrentController?, RemoveScore[MaybeCurrentController?]):

        if:
            AddScore[InAgent]
            AgentColor := AgentToColorMap[InAgent]?
        then:
            GameplayCubeRef.SetController(InAgent)
            GameplayCubeRef.SetColor(AgentColor)


        # Update UI after all checks and score updating, we know the first agent in the AgentToScoreMap is the red one. 
        if:
            UIManagerEntity := MaybeUIManager?
            UIManager := UIManagerEntity.GetComponent[ui_manager_component]
            var MaybeRedAgentScore: ?int = false
            var MaybeBlueAgentScore: ?int = false
            var Index: int = 0
        then:
            for (Key->Value : AgentToScoreMap):
                if (Index = 0):
                    set MaybeRedAgentScore = Value
                else:
                    set MaybeBlueAgentScore = Value

                set Index += 1

            if (RedScore := MaybeRedAgentScore?, BlueScore := MaybeBlueAgentScore?):
                UIManager.UpdateTexts(RedScore, BlueScore, CurrentGameTimer)


# Reset variables for new game and display winner.
(Manager:game_manager_component).Reset<public>():void=
    if(MessageDevice := Manager.MaybeMessageDevice?):
        var MaxValue:int = 0
        var WinnerAgent:?agent = false
        for(Key -> Value : Manager.AgentToScoreMap):
            if(Score := Value?, Score > MaxValue):
                set MaxValue = Score
                set WinnerAgent = option. Key
        if(Winner := WinnerAgent?):
            MessageDevice.SetText(EndGameMessage(Winner,MaxValue))
            MessageDevice.Show()
    for (Agent->Color : Manager.AgentToColorMap, set Manager.AgentToColorMap[Agent] = false):
    for (Agent->Score : Manager.AgentToScoreMap, set Manager.AgentToScoreMap[Agent] = false):
    for (Cube : Manager.GameplayCubes) do Cube.Reset()
    for (MutatorZone : Manager.MutatorZones) do MutatorZone.Disable()
    
    set Manager.CurrentGameTimer = Manager.GameLength

    if. UIManager := Manager.MaybeUIManager?.GetComponent[ui_manager_component] then. UIManager.RemoveUI()
    if:
        Sim := Manager.Entity.GetSimulationEntity[]
        StartGameComponents := Sim.FindDescendantComponents(start_game_component)
    then:
        for(StartGameComp : StartGameComponents) do StartGameComp.Reset()

# Initialize variables for new game.
# By default, P1Agent is the red player and P2Agent is the blue player.
(Manager:game_manager_component).InitializePlayers<public>(P1Agent:agent, P2Agent:agent)<decides><transacts>:void=
    set Manager.AgentToColorMap[P1Agent] = option{Manager.P1Color}
    set Manager.AgentToColorMap[P2Agent] = option{Manager.P2Color}

    set Manager.AgentToScoreMap[P1Agent] = option{0}
    set Manager.AgentToScoreMap[P2Agent] = option{0}


# Add score to InAgent.
(Manager:game_manager_component).AddScore<public>(InAgent:agent)<decides><transacts>:void=
        CurrentScore: int = Manager.AgentToScoreMap[InAgent]?
        set Manager.AgentToScoreMap[InAgent] = option{CurrentScore + 1}


# Remove score from InAgent.
(Manager:game_manager_component).RemoveScore<public>(InAgent:agent)<decides><transacts>:void=
    CurrentScore: int = Manager.AgentToScoreMap[InAgent]?
    set Manager.AgentToScoreMap[InAgent] = option{CurrentScore - 1}


# Start game by enabling mutator zones and cube movement.
(Manager:game_manager_component).StartGame<public>():void=
    for (MutatorZone : Manager.MutatorZones) do MutatorZone.Enable()
    for (Cube : Manager.GameplayCubes) do Cube.EnableAnimation()

    spawn{Manager.RunGame()}


# Async function that runs until the game timer is 0.
(Manager:game_manager_component).RunGame<public>()<suspends>:void=
    loop:
        if (Manager.CurrentGameTimer <= 0):
            Manager.Reset()
            return

        Sleep(1.0)
        set Manager.CurrentGameTimer -= 1

        if:
            UIManagerEntity := Manager.MaybeUIManager?
            UIManager := UIManagerEntity.GetComponent[ui_manager_component]
        then:
            UIManager.UpdateGameTimer(Manager.CurrentGameTimer)


# === Messages ===


GameLengthToolTip<localizes>: message       = "Game length, in seconds."
GameplayCubesToolTip<localizes>: message    = "Collection of Gameplay Cubes for the minigame."
MutatorZonesToolTip<localizes>: message     = "Collection of Mutator Zones for the minigame."
UIManagerToolTip<localizes>: message        = "UI manager reference."
MessageDevicerToolTip<localizes>: message   = "hud_message_device reference."
P1ColorToolTip<localizes>: message          = "Color used to signify player 1's control of a Gameplay Cube."
P2ColorToolTip<localizes>: message          = "Color used to signify player 2's control of a Gameplay Cube."

GameplayCategory<localizes>: message            = "Gameplay Details"
PlayerCustomizationCategory<localizes>: message = "Player Customization"

EndGameMessage<localizes>(Winner:agent,Score:int):message="{Winner} won with {Score} points!"