<#
@UUID: 3fe5ff02-2eba-4610-9729-2324a59b6bcd
@Task: Implement the gameplay cube's core logic.
@Description: Make a SceneGraph component of a cube that changes player control when touched, and changes the color of a child entity's static mesh.
@Tags: Gameplay
@Export: true
#>

using { /Verse.org/Colors }
using { /Verse.org/Random }
using { /Verse.org/SceneGraph }
using { /Verse.org/SceneGraph/KeyframedMovement }
using { /Verse.org/Simulation }
using { /Verse.org/SpatialMath }




# Custom function type for ChangeControllerSubscribable.
change_controller_function := type{_(:agent, :?agent, :Prefabs.gameplay_cube):void}


# Class for custom subscribable.
change_controller_subscribable := class:
    var Functions<private>:[]change_controller_function = array{}


    # Subscribe-like function.
    # Needs to be wrapped by gameplay_cube_component.
    # Accepts functions of type change_controller_function, which will be run during Dispatch().
    AddFunction(InFunction:change_controller_function):void=
        set Functions += array{InFunction}


    # Call all functions registered.
    Dispatch(InAgent:agent, MaybeCurrentController:?agent, GameplayCubeRef:Prefabs.gameplay_cube):void=
        for (Function : Functions):
            Function(InAgent, MaybeCurrentController, GameplayCubeRef)


# Component for gameplay cube's core logic.
gameplay_cube_component<public> := class<final_super>(component):
    @editable:
        ToolTip := KeyframesToolTip
        Categories := array{KeyframesCategory}
    var Keyframes: []keyframed_movement_delta = array{}

    var MaybeCurrentController: ?agent = false

    ChangeControllerSubscribable: change_controller_subscribable = change_controller_subscribable{}

    @editable
    DefaultColor: color = NamedColors.Yellow
    var MaybeStartPos:?transform = false

    OnBeginSimulation<override>():void=
        (super:)OnBeginSimulation()
        set MaybeStartPos = option. Entity.GetGlobalTransform()


        # Subscribe to entity collision.
        for:
            DescendantEntity : Entity.FindDescendantEntitiesWithComponent(mesh_component)
            OuterMeshComponent := DescendantEntity.GetComponent[mesh_component]
        do:
            OuterMeshComponent.EntityEnteredEvent.Subscribe(OnEntityEntered)


    # === Entity Functions ===


    # Handle an agent's collision proxy hitting the mesh.
    OnEntityEntered(InEntity:entity):void=
        if:
            CharacterCollisionProxy := Prefabs.character_collision_proxy[InEntity]
            Agent := CharacterCollisionProxy.GetAgent[] 
            GameplayCubeRef := Prefabs.gameplay_cube[Entity]
        then:
            ChangeControllerSubscribable.Dispatch(Agent, MaybeCurrentController, GameplayCubeRef)


    # === Gameplay Cube Functions ===
    

# Reset variables for new game.
(GameplayCube:Prefabs.gameplay_cube).Reset<public>():void=
    if (GameplayCubeComponent := GameplayCube.GetComponent[gameplay_cube_component], StartPos := GameplayCubeComponent.MaybeStartPos?):
        set GameplayCubeComponent.MaybeCurrentController = false
        GameplayCube.SetColor(GameplayCubeComponent.DefaultColor)
        GameplayCube.SetGlobalTransform(StartPos)

    if (KFM: keyframed_movement_component = GameplayCube.GetComponent[keyframed_movement_component]):
        KFM.Stop()


# Start the animation for the cube.
(GameplayCube:Prefabs.gameplay_cube).EnableAnimation<public>():void=
        if (KFM: keyframed_movement_component = GameplayCube.GetComponent[keyframed_movement_component]):
            KFM.SetKeyframes(GameplayCube.GetKeyframes(), pingpong_keyframed_movement_playback_mode{})
            KFM.Play()


# Get the keyframes for this GameplayCube.
(GameplayCube:Prefabs.gameplay_cube).GetKeyframes<public>():[]keyframed_movement_delta=
    if (GameplayCubeComponent := GameplayCube.GetComponent[gameplay_cube_component]):
        array{GenerateKeyframe(), GenerateKeyframe()}
    else:
        array{}


# Set the controller of the cube.
(GameplayCube:Prefabs.gameplay_cube).SetController<public>(InAgent:agent):void=
    if (GameplayCubeComponent := GameplayCube.GetComponent[gameplay_cube_component]):
        set GameplayCubeComponent.MaybeCurrentController = option{InAgent}


# Subscribe to ChangeController custom event.
(GameplayCube:Prefabs.gameplay_cube).Subscribe<public>(InFunction:change_controller_function):void=
    if (GameplayCubeComponent := GameplayCube.GetComponent[gameplay_cube_component]):
        GameplayCubeComponent.ChangeControllerSubscribable.AddFunction(InFunction)


# Set the color of the inner cube.
(GameplayCube:Prefabs.gameplay_cube).SetColor<public>(InColor:color):void=
    CubeInnerComps := GameplayCube.FindDescendantComponents(Components.inner_cube_component)

    for (CubeInnerComp : CubeInnerComps):
        CubeInnerComp.SetColor(InColor)


# Generate a keyframe to move back and forth between.
GenerateKeyframe():keyframed_movement_delta=
    keyframed_movement_delta:
        Transform := transform:
            Translation := vector3:
                Forward := GetRandomFloat(-800.0, 800.0)
                Left    := GetRandomFloat(-800.0, 800.0)
                Up      := GetRandomFloat(-200.0, 200.0)

            Rotation := IdentityRotation()
            
            Scale := vector3:
                Forward := 0.0
                Left    := 0.0
                Up      := 0.0

        Duration := GetRandomFloat(2.0, 5.0)

        Easing := ease_in_out_cubic_bezier_easing_function{}


# === Messages ===


KeyframesToolTip<localizes>:message  = "Keyframes for the cube to move between."
KeyframesCategory<localizes>:message = "Keyframes"