<#
@UUID: a285661a-73f0-4026-9255-e98fb18f1389
@Task: Implement the inner-cube mesh logic.
@Description: Make a SceneGraph component that controls the color of a cube.
@Tags: Gameplay
@Export: true
#>

using {/Verse.org/Colors}
using {/Verse.org/SceneGraph}

# Component that controls the gameplay_cube's inner mesh.
inner_cube_component<public> := class<final_super>(component):


    # === Entity Functions ===


    OnBeginSimulation<override>():void=
        (super:)OnBeginSimulation()


    # Set the inner_cube's material to InColor.
    SetColor<public>(InColor:color):void=
        if:
            CubeMeshComponent := Entity.GetComponent[mesh_component]
            CubeInnerMesh := Meshes.SM_CubeInner[CubeMeshComponent]
        then:
            NewMat: Materials.M_CubeInner = Materials.M_CubeInner{}
            set NewMat.BaseColor = InColor

            set CubeInnerMesh.Material = NewMat