<#
@UUID: 0fdce248-e9cb-46c3-929b-10f0ca351dd9
@Task: A component that functions as a respawn point.
@Description: Implement a SceneGraph component that will act as a respawn point.
@Tags: Respawn, Gameplay
@Export: true
#>

using { /Verse.org/SceneGraph }
using { /Verse.org/Simulation }
using { Prefabs }

respawn_component<public> := class<final_super>(component):

    var MaybeMesh:?mesh_component = false
    var SpawnerAvailable<public>:logic = false
    var AgentsInside:[]agent = array{}

    OnBeginSimulation<override>():void=
        if(Mesh := Entity.GetComponent[mesh_component]):
            set MaybeMesh = option. Mesh
            Mesh.EntityEnteredEvent.Subscribe(OnEntityEntered)
            Mesh.EntityExitedEvent.Subscribe(OnEntityExited)

    OnEntityEntered(EnteredEntity:entity):void=
        if:
            Sim := Entity.GetSimulationEntity[]
            Proxy := character_collision_proxy[EnteredEntity] 
            Agent := Proxy.GetAgent[]
        then:
            set AgentsInside += array{Agent}

    OnEntityExited(EnteredEntity:entity):void=
        if:
            Sim := Entity.GetSimulationEntity[]
            Proxy := character_collision_proxy[EnteredEntity] 
            Agent := Proxy.GetAgent[]
        then:
            var Index:int = 0
            for(Elem : AgentsInside):
                if:
                    Elem = Agent
                    NewArray := AgentsInside.RemoveElement[Index]
                then:
                    set AgentsInside = NewArray
                set Index += 1

    IsAvailable<public>()<decides><transacts>:void=
        AgentsInside.Length = 0