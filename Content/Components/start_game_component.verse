<#
@UUID: 87daeba0-6704-45b5-8d85-5b7358b2ca08
@Task: Make a component that will contain logic to start the game.
@Description: Implement a SceneGraph component that will manage the game start.
@Tags: GameManagement
@Export: true
#>

using { /Fortnite.com/Characters }
using { /Fortnite.com/Playspaces }
using { /Fortnite.com/UI }
using { /UnrealEngine.com/Temporary/SpatialMath }
using { /UnrealEngine.com/Temporary/UI }
using { /Verse.org/SceneGraph }
using { /Verse.org/Simulation }

start_game_component<public> := class<final_super>(component):
    
    @editable
    MaybeRedSpawnPosition:?entity = false
    @editable
    MaybeBlueSpawnPosition:?entity = false
    Countdown:float = 3.0
    var GameStarted<public>:logic = false
    var MaybeRedPlayer:?agent = false
    var MaybeBluePlayer:?agent = false
    var RedPlayerReady:logic = false
    var BluePlayerReady:logic = false
    var PlayersUI:[agent]timer_ui = map{}


    # Funcion to init UI adding it to each player
    InitUI(Player:player):void=
        if: 
            set PlayersUI[Player] = timer_ui{}
            PlayerCanvas := PlayersUI[Player]
            PlayerUI := GetPlayerUI[Player]
        then:
            PlayerUI.AddWidget(PlayerCanvas.Canvas)


    # Funcion to remove UI form players
    RemoveUI(Player:player):void=
        if:
            PlayerCanvas := PlayersUI[Player]
            PlayerUI := GetPlayerUI[Player]
        then. PlayerUI.RemoveWidget(PlayerCanvas.Canvas)


    OnBeginSimulation<override>():void=
        (super:)OnBeginSimulation()
        if (RedEntity := MaybeRedSpawnPosition?, BlueEntity := MaybeBlueSpawnPosition?):
            Print("Spawns correctly set up")
        else:
            Err("Spawns entity reference missing.")
        if (not GameStarted?):
            # Check interaction with the starting buttons
            Interactables := for (Interactable:Entity.FindDescendantEntitiesWithComponent(game_start_component), Comp := Interactable.GetComponent[game_start_component]). Comp
            if(Interactable := Interactables[0]): 
                Interactable.SucceededEvent.Subscribe(OnGameStartButtonRedInteractedWith)
            if(Interactable := Interactables[1]): 
                Interactable.SucceededEvent.Subscribe(OnGameStartButtonBlueInteractedWith)


    # Set red player
    OnGameStartButtonRedInteractedWith(Agent:agent):void=
        Print("Red Player selected")
        set MaybeRedPlayer = option. Agent
        set RedPlayerReady = true
        CheckGameStart()


    # Set blue player
    OnGameStartButtonBlueInteractedWith(Agent:agent):void=
        Print("Blue Player selected")
        set MaybeBluePlayer = option. Agent
        set BluePlayerReady = true
        CheckGameStart()


    CheckGameStart():void=
        if (not (RedPlayerReady? and BluePlayerReady?)):
            return

        if:
            RedSpawnTransform := MaybeRedSpawnPosition?.GetGlobalTransform()
            BlueSpawnTransform := MaybeBlueSpawnPosition?.GetGlobalTransform()
            
            BluePlayer := MaybeBluePlayer?
            RedPlayer := MaybeRedPlayer?
            RedFort := RedPlayer.GetFortCharacter[]
            BlueFort := BluePlayer.GetFortCharacter[]

            SimEntity := Entity.GetSimulationEntity[]

            # Teleport players to their starting positions
            BlueFort.TeleportTo[
                (/UnrealEngine.com/Temporary/SpatialMath:)vector3{
                    X := BlueSpawnTransform.Translation.Forward, 
                    Y := -BlueSpawnTransform.Translation.Left, 
                    Z := BlueSpawnTransform.Translation.Up + 300.0}, 
                (/UnrealEngine.com/Temporary/SpatialMath:)rotation{}]

            RedFort.TeleportTo[
                (/UnrealEngine.com/Temporary/SpatialMath:)vector3{
                    X := RedSpawnTransform.Translation.Forward, 
                    Y := -RedSpawnTransform.Translation.Left, 
                    Z := RedSpawnTransform.Translation.Up + 300.0}, 
                (/UnrealEngine.com/Temporary/SpatialMath:)rotation{}]
        then:
            set GameStarted = true
            spawn. StartCountdown(RedFort, BlueFort)

            # Update the game_manager_component with these players.
            GameManagerComponents := SimEntity.FindDescendantComponents(Components.game_manager_component)
            UIManagers := SimEntity.FindDescendantComponents(Components.ui_manager_component)
            for (GameManager : GameManagerComponents, GameManager.InitializePlayers[RedPlayer, BluePlayer]):
            for(UIManager : UIManagers):
                UIManager.Init(BluePlayer,RedPlayer)

        else Print("Error: Spawn positions not setup on the component.")


    # Block the players on the starting position and start the countdown
    StartCountdown(RedFort:fort_character, BlueFort:fort_character)<suspends>:void=
        if (not GameStarted?):
            return

        RedFort.PutInStasis(stasis_args{AllowTurning := true})
        BlueFort.PutInStasis(stasis_args{AllowTurning := true})
        if. BluePlayer := player[MaybeBluePlayer?] then. InitUI(BluePlayer)
        if. RedPlayer := player[MaybeRedPlayer?] then. InitUI(RedPlayer)
        if. var counter:int = Int[Countdown]
        then. CountdownTick(counter, RedFort, BlueFort)


    # Countdown function, decrease a value and when ends the game starts
    CountdownTick(TimeLeft:int, RedFort:fort_character, BlueFort:fort_character)<suspends>:void=
        if (not GameStarted?):
            return

        # Update UIs
        if (BlueUI := PlayersUI[BlueFort.GetAgent[]], RedUI := PlayersUI[RedFort.GetAgent[]]):
            BlueUI.UpdateTimerUI(TimeLeft)
            RedUI.UpdateTimerUI(TimeLeft)

        # Start game if TimeLeft is <= 0
        if (TimeLeft <= 0, SimEntity := Entity.GetSimulationEntity[]):
            RedFort.ReleaseFromStasis()
            BlueFort.ReleaseFromStasis()

            # Remove Countdown UI from players
            Sleep(1.0)
            if. BluePlayer := player[MaybeBluePlayer?] then. RemoveUI(BluePlayer)
            if. RedPlayer := player[MaybeRedPlayer?] then. RemoveUI(RedPlayer)

            GameManagerComponents := SimEntity.FindDescendantComponents(Components.game_manager_component)
            for (GameManager : GameManagerComponents):
                GameManager.StartGame()
            
            return

        # Tick down another second
        Sleep(1.0)
        CountdownTick(TimeLeft - 1, RedFort, BlueFort)


    # Set logics false when race ends
    Reset<public>():void=
        set GameStarted = false
        set RedPlayerReady = false
        set BluePlayerReady = false
        set MaybeRedPlayer = false
        set MaybeBluePlayer = false

    # When simultaion ends remove UI
    OnEndSimulation<override>():void=
        for (Player : Entity.GetPlayspaceForEntity[].GetPlayers()):
            RemoveUI(Player)


# Subclassing interactable_component to exactly find this component in the scene.
game_start_component<public> := class(interactable_component):


# Messages for countdown UI
MessageTextInt<localizes>(TimeLeft:int):message= "{TimeLeft}"
MessageTextChar<localizes>:message= "START"

timer_ui := class():
    var CountdownTextBlock:text_block = text_block:
        DefaultJustification := text_justification.Center
    var Canvas<internal>: canvas = canvas{}
    block:
        set Canvas = canvas:
            Slots := array{
                canvas_slot:
                    Widget := CountdownTextBlock
                    Anchors := anchors{Minimum:=vector2{X:=0.5, Y:=0.5}, Maximum:=vector2{X:=0.5, Y:=0.5}}
                    Offsets := margin{Left:=0.0, Top:=-350.0}
            }   
    UpdateTimerUI<public>(TimeLeft:int):void=
        if(TimeLeft > 0):
            CountdownTextBlock.SetText(MessageTextInt(TimeLeft))
        else:
            CountdownTextBlock.SetText(MessageTextChar)