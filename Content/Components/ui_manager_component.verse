<#
@UUID: ab82f3eb-def1-4c1e-909d-7f822dd9a9df
@Task: A component to manage the UI of both players.
@Description: Implement a SceneGraph component that will manage the UI of both players.
@Tags: UI
@Export: true
#>

using { /Fortnite.com/Characters }
using { /Fortnite.com/Devices }
using { /Fortnite.com/Playspaces }
using { /Fortnite.com/UI }
using { /UnrealEngine.com/Temporary/UI }
using { /Verse.org/Colors }
using { /Verse.org/Random }
using { /Verse.org/SceneGraph }
using { /Verse.org/Simulation }
using { /Verse.org/Simulation/Tags }
using { /UnrealEngine.com/Temporary/SpatialMath }
using { /Verse.org/SpatialMath }
using { Prefabs }

ui_manager_component<public> := class<final_super>(component):

    var MaybeRedPlayer:?agent = false
    var MaybeBluePlayer:?agent = false
    var PlayersUI:[agent]agent_ui = map{}

    OnEndSimulation<override>():void=
        if:
            BluePlayer := MaybeBluePlayer?
            RedPlayer := MaybeRedPlayer?
        then:
            RemoveUI(BluePlayer)
            RemoveUI(RedPlayer)

    Init(BluePlayer:agent, RedPlayer:agent):void=
        set MaybeBluePlayer = option. BluePlayer
        set MaybeRedPlayer = option. RedPlayer
        if. set PlayersUI[RedPlayer] = agent_ui{}
        if. set PlayersUI[BluePlayer] = agent_ui{}
        # Generating and adding UI to both players
        GenerateUI()

    RemoveUI<public>(Agent:agent):void=
        if(AgentUI := PlayersUI[Agent], PlayerUI := GetPlayerUI[player[Agent]]):
            PlayerUI.RemoveWidget(AgentUI.Canvas)

    RemoveUI<public>():void=
        if:
            BluePlayer := MaybeBluePlayer?
            RedPlayer := MaybeRedPlayer?
        then:
            RemoveUI(BluePlayer)
            RemoveUI(RedPlayer)

    GenerateUI():void=
        UpdateTexts(0, 0, 0)
        if(BluePlayer := MaybeBluePlayer?, RedPlayer := MaybeRedPlayer?):
            if(BluePlayerUI := GetPlayerUI[player[BluePlayer]], RedPlayerUI := GetPlayerUI[player[RedPlayer]]):
                if(BlueUI := PlayersUI[BluePlayer], RedUI := PlayersUI[RedPlayer]):
                    RedPlayerUI.AddWidget(RedUI.Canvas)
                    BluePlayerUI.AddWidget(BlueUI.Canvas)

    UpdateTexts<public>(RedScore:int, BlueScore:int, GameTime:int):void=
        if(RedAgent := MaybeRedPlayer?, BlueAgent := MaybeBluePlayer?):
            if(BlueUI := PlayersUI[BlueAgent], RedUI := PlayersUI[RedAgent]):
                BlueUI.UpdateUI(RedAgent, RedScore, BlueAgent, BlueScore, GameTime)
                RedUI.UpdateUI(RedAgent, RedScore, BlueAgent, BlueScore, GameTime)

    UpdateGameTimer<public>(GameTime:int):void=
        if:
            RedAgent := MaybeRedPlayer? 
            BlueAgent := MaybeBluePlayer?
            BlueUI := PlayersUI[BlueAgent]
            RedUI := PlayersUI[RedAgent]
        then:
            BlueUI.UpdateGameTimer(GameTime)
            RedUI.UpdateGameTimer(GameTime)


agent_ui := class():
    RedPlayerTextBlock:text_block = text_block{}
    BluePlayerTextBlock:text_block = text_block{}
    TimerTextBlock:text_block = text_block{}

    var Canvas<internal>:canvas = canvas{}
    block:
        set Canvas = canvas:
            Slots := array{
                canvas_slot{
                    Widget := RedPlayerTextBlock
                    Anchors := anchors{Minimum := vector2{X := 0.4, Y := 0.0}, Maximum := vector2{X := 0.4, Y := 0.0}}
                    Offsets := margin{Top := 75.0}
                    SizeToContent := true
                },
                canvas_slot{
                    Widget := BluePlayerTextBlock
                    Anchors := anchors{Minimum := vector2{X := 0.4, Y := 0.0}, Maximum := vector2{X := 0.4 , Y := 0.0}}
                    Offsets := margin{Top := 150.0}
                    SizeToContent := true
                },
                canvas_slot{
                    Widget := TimerTextBlock
                    Anchors := anchors{Minimum := vector2{X := 0.4, Y := 0.0}, Maximum := vector2{X := 0.4 , Y := 0.0}}
                    SizeToContent := true
                }
            }

    UpdateUI<public>(RedAgent:agent, RedScore:int,BlueAgent:agent, BlueScore:int, GameTimer:int):void=
        RedPlayerTextBlock.SetText(PlayerScoreMessage(RedAgent, RedScore))
        BluePlayerTextBlock.SetText(PlayerScoreMessage(BlueAgent, BlueScore))
        TimerTextBlock.SetText(GameTimeMessage(GameTimer))

    UpdateGameTimer<public>(GameTimer:int):void=
        TimerTextBlock.SetText(GameTimeMessage(GameTimer))


PlayerScoreMessage<localizes>(Agent:agent,Score:int):message="{Agent}: {Score}"
GameTimeMessage<localizes>(Time:int):message="Game Time: {Time}"