<#
@UUID: c46fc57e-14a5-4dc2-9c7c-87e67e01fbe5
@Task: Editor Entites Prefab Setup
@Description: A file containing detailed information about how to create and setup the entity prefabs needed for the game. 
@Tags: Editor, Setup.
@Export: true
#>

<#
In Editor character_collision_proxy Entity Prefab Setup:
- character_collision_proxy
    - transform_component
    - mesh_component
        - Visible = false
    - character_collision_proxy_component
#>

<#
In Editor collision_proxy_spawner Entity Prefab Setup:
- collision_proxy_spawner
    - collision_proxy_spawner_component
    - transform_component
#>

<#
In Editor death_volume Entity Prefab Setup:
- death_volume
    - transform_component
    - mesh_component
        - CollisionPreset = OverlapAllDynamic
    - death_volume_component
#>

<#
In Editor start_game Entity Prefab Setup:
- start_game
    - transform_component
    - start_game_component
        - MaybeRedSpawnPosition = red_starting_plate
        - MaybeBlueSpawnPosition = blue_starting_plate
    - blue_starting_plate (entity)
        - mesh_component
        - transform_component
    - blue_interactable (entity)
        - transform_component
        - mesh_component
        - game_start_component
    - red_starting_plate (entity)
        - mesh_component
        - transform_component
    - red_interactable (entity)
        - transform_component
        - mesh_component
        - game_start_component
#>

<#
In Editor ui_manager Entity Prefab Setup:
- ui_manager
    - transform_component
    -ui_manager_component
#>

<#
In Editor game_manager Entity Prefab Setup:
    - game_manager
        - game_manager_component
#>

<#
In Editor gameplay_cube Entity Prefab Setup:
    - gameplay_cube
        - transform_component
        - gameplay_cube_component
        - SM_CubeOuter (mesh_component)

        - cube_inner (entity)
            - transform_component
            - inner_cube_component
            - SM_CubeInner (mesh_component)
#>

<#
In Editor respawn Entity Prefab Setup:
    - respawn
        - SM_Respawn (mesh_component)
        - respawn_component
        - transform_component
#>