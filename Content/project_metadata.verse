<#
@UUID: 098e7147-d77c-4073-9bb6-14a86c7621c5
@Task: Spacewalk.
@Description: A minigame where two players challenge each other jumping on moving cubes in a low-gravity space.
@Tags: Minigame.
@Export: true
#>